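# Sound cue for the reaction sound of the HumanBeing back bling cosmetic.
# A Random node picks one of three React wave variations with equal weight.
# It plays through the back bling sound class with close-range pet attenuation, a reactive back bling audio property sheet and a reduced volume.
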SoundCue /BRCosmetics/Sounds/Backblings/HumanBeing/Backpack_HumanBeing_React_Cue
root=Random_0
nodes [
 {id=Random_0, children [WavePlayer_2,WavePlayer_3,WavePlayer_4], properties {weights [1,1,1]}}
 {id=WavePlayer_2, properties {soundWaveAssetPtr=&/Backpack_HumanBeing_React_01}}
 {id=WavePlayer_3, properties {soundWaveAssetPtr=&/Backpack_HumanBeing_React_02}}
 {id=WavePlayer_4, properties {soundWaveAssetPtr=&/Backpack_HumanBeing_React_03}}
]
properties {
 attenuationSettings=&Fort_Pets_Close_Athena_Att
 audioPropertiesSheet=/BRCosmetics/Sounds/Settings/PropertySheets/Reactive/APS_Cosmetic_Reactive_Backbling
 maxDistance=1800.0001
 soundClassObject=/Game/Sounds/FortSoundClassesAndModes/Fort_Backblings
 volumeMultiplier=0.65
}
